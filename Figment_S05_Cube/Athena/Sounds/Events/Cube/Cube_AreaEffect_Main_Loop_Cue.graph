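# Looping Sound Cue for the Cube's area effect, by its name the main loop layer.
# A Quality Level node feeds the first two quality slots with the looping Cube_Print_01 wave and the last two with its "_opt" variant, likely an optimized version.
# An Attenuation node at the root applies close-range Cube ambient falloff.
# A concurrency override limits the cue to one instance, routed through an important in-game mission sound class.
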SoundCue /Figment_S05_Cube/Athena/Sounds/Events/Cube/Cube_AreaEffect_Main_Loop_Cue
root=Attenuation_1
nodes [
 {
  id=Attenuation_1
  children [QualityLevel_0]
  properties {attenuationSettings=&/Cube_Ambient_Close_Att}
 }
 {id=QualityLevel_0, children [WavePlayer_1,WavePlayer_1,WavePlayer_3,WavePlayer_3]}
 {id=WavePlayer_1, properties {bLooping=true, soundWaveAssetPtr=&/Cube_Print_01}}
 {id=WavePlayer_3, properties {bLooping=true, soundWaveAssetPtr=&/Cube_Print_01_opt}}
]
properties {
 bOverrideConcurrency=true
 maxDistance=3584
 soundClassObject=&Fort_Mission_Ingame_Important_PSM
 volumeMultiplier=1
 concurrencyOverrides {maxCount=1}
}
