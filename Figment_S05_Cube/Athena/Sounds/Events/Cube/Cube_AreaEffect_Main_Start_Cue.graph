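# Sound cue for the start of the Cube's main area effect in the Figment S05 Cube event content.
# A single Cube_PrintBegin wave player feeds an attenuation node using the Cube print vocal settings, then a one-input mixer that halves its volume.
# It uses an important in-game mission sound class, has a long audible range, and overrides concurrency to allow only one instance at a time.
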SoundCue /Figment_S05_Cube/Athena/Sounds/Events/Cube/Cube_AreaEffect_Main_Start_Cue
root=Mixer_0
nodes [
 {id=Attenuation_2, children [WavePlayer_4], properties {attenuationSettings=&/Cube_Print_Vocal_Att}}
 {id=Mixer_0, children [Attenuation_2], properties {inputVolume [0.5]}}
 {id=WavePlayer_4, properties {soundWaveAssetPtr=/Game/Athena/Sounds/Events/Cube/Cube_PrintBegin_01}}
]
properties {
 bOverrideConcurrency=true
 maxDistance=30000
 soundClassObject=&Fort_Mission_Ingame_Important_PSM
 volumeMultiplier=1
 concurrencyOverrides {maxCount=1}
}
