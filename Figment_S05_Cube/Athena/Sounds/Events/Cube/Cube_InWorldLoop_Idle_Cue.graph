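# Sound cue for the Cube's idle in-world ambient loop, limited to one concurrent instance.
# A quality-level switch picks standard or "_opt" versions of two looping beds, ambient and close, each with its own attenuation and mixed with the close layer louder.
# Alongside the beds, a looping branch with area-effect attenuation plays one of two random PrintVoice one-shots after a random 5–15 second delay.
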SoundCue /Figment_S05_Cube/Athena/Sounds/Events/Cube/Cube_InWorldLoop_Idle_Cue
root=Mixer_2
nodes [
 {id=Attenuation_0, children [WavePlayer_2], properties {attenuationSettings=&/Cube_Ambient_Att}}
 {
  id=Attenuation_1
  children [WavePlayer_0]
  properties {attenuationSettings=&/Cube_Ambient_Close_Att}
 }
 {id=Attenuation_2, children [WavePlayer_4], properties {attenuationSettings=&Cube_Ambient_Att}}
 {
  id=Attenuation_3
  children [WavePlayer_3]
  properties {attenuationSettings=&Cube_Ambient_Close_Att}
 }
 {id=Attenuation_5, children [Looping_0], properties {attenuationSettings=/Game/Athena/Sounds/Events/Cube/Cube_AreaEffect_Att}}
 {id=Delay_0, children [Random_0], properties {delayMax=15, delayMin=5}}
 {id=Looping_0, children [Delay_0]}
 {id=Mixer_0, children [Attenuation_0, Attenuation_1], properties {inputVolume [0.25,1.1]}}
 {id=Mixer_1, children [Attenuation_2, Attenuation_3], properties {inputVolume [0.25,1.1]}}
 {id=Mixer_2, children [QualityLevel_0, Attenuation_5], properties {inputVolume [1,1]}}
 {id=QualityLevel_0, children [Mixer_0,Mixer_0,Mixer_1,Mixer_1]}
 {id=Random_0, children [WavePlayer_5,WavePlayer_6], properties {weights [1,1]}}
 {id=WavePlayer_0, properties {bLooping=true, soundWaveAssetPtr=/Game/Athena/Sounds/Events/Cube/Cube_AmbientClose_01}}
 {id=WavePlayer_2, properties {bLooping=true, soundWaveAssetPtr=/Game/Athena/Sounds/Events/Cube/Cube_Ambient_01}}
 {id=WavePlayer_3, properties {bLooping=true, soundWaveAssetPtr=/Game/Athena/Sounds/Events/Cube/Cube_AmbientClose_01_opt}}
 {id=WavePlayer_4, properties {bLooping=true, soundWaveAssetPtr=/Game/Athena/Sounds/Events/Cube/Cube_Ambient_01_opt}}
 {id=WavePlayer_5, properties {soundWaveAssetPtr=/Game/Athena/Sounds/Events/Cube/Cube_PrintVoice_02}}
 {id=WavePlayer_6, properties {soundWaveAssetPtr=/Game/Athena/Sounds/Events/Cube/Cube_PrintVoice_01}}
]
properties {
 bOverrideConcurrency=true
 maxDistance=10240
 soundClassObject=&Fort_Mission_Ingame_Important_PSM
 volumeMultiplier=0.7
 concurrencyOverrides {maxCount=1}
}
