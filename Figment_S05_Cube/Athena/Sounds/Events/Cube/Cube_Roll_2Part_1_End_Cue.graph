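# Sound cue for the end of part one of a two-part cube roll, playing the cube's impact sound.
# A mixer layers a close impact wave and a distant impact wave, each through its own attenuation, with the distant layer weighted louder.
# The cue uses an important in-game mission sound class and overrides concurrency to allow only one instance at a time.
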SoundCue /Figment_S05_Cube/Athena/Sounds/Events/Cube/Cube_Roll_2Part_1_End_Cue
root=Mixer_1
nodes [
 {
  id=Attenuation_2
  children [WavePlayer_8]
  properties {attenuationSettings=&/Cube_Roll_Impact_Close_Att}
 }
 {
  id=Attenuation_3
  children [WavePlayer_9]
  properties {attenuationSettings=&/Cube_Roll_Impact_Distant_Att}
 }
 {id=Mixer_1, children [Attenuation_2, Attenuation_3], properties {inputVolume [1,2]}}
 {id=WavePlayer_8, properties {soundWaveAssetPtr=&/Cube_Impact2Part_01}}
 {id=WavePlayer_9, properties {soundWaveAssetPtr=&/Cube_Impact2PartDistant_01}}
]
properties {
 bOverrideConcurrency=true
 maxDistance=30000
 soundClassObject=&Fort_Mission_Ingame_Important_PSM
 volumeMultiplier=0.7
 concurrencyOverrides {maxCount=1}
}
