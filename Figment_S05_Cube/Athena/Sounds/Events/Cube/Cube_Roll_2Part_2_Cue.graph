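# Sound cue named for part two of a cube roll event sound in Figment Season 5, currently holding no nodes and no root, so it plays no audio.
# It keeps only playback settings: the important in-game mission sound class, reduced volume and a maximum distance.
# A concurrency override limits it to a single simultaneous instance.
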SoundCue /Figment_S05_Cube/Athena/Sounds/Events/Cube/Cube_Roll_2Part_2_Cue
root=null
nodes []
properties {
 bOverrideConcurrency=true
 maxDistance=10000
 soundClassObject=&Fort_Mission_Ingame_Important_PSM
 volumeMultiplier=0.7
 concurrencyOverrides {maxCount=1}
}
