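# Sound cue for the rolling sound of the event Cube, limited to one concurrent instance.
# A mixer layers a random pick of two roll waves with a debris wave mixed in louder.
# Each layer has its own attenuation, with a close-range asset for the debris.
# A quality-level node supplies the debris only in its first two slots, leaving the other two empty.
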SoundCue /Figment_S05_Cube/Athena/Sounds/Events/Cube/Cube_Roll_Cue
root=Mixer_0
nodes [
 {id=Attenuation_0, children [Random_0], properties {attenuationSettings=/Game/Athena/Sounds/Events/Cube/Cube_Roll_Att}}
 {id=Attenuation_1, children [WavePlayer_2], properties {attenuationSettings=/Game/Athena/Sounds/Events/Cube/Cube_Roll_Close}}
 {id=Mixer_0, children [Attenuation_0, QualityLevel_0], properties {inputVolume [1,1.5]}}
 {id=QualityLevel_0, children [Attenuation_1, Attenuation_1, null, null]}
 {id=Random_0, children [WavePlayer_0,WavePlayer_1], properties {weights [1,1]}}
 {id=WavePlayer_0, properties {soundWaveAssetPtr=/Game/Athena/Sounds/Events/Cube/Cube_Roll_01}}
 {id=WavePlayer_1, properties {soundWaveAssetPtr=/Game/Athena/Sounds/Events/Cube/Cube_Roll_02}}
 {id=WavePlayer_2, properties {soundWaveAssetPtr=/Game/Athena/Sounds/Events/Cube/Cube_RollDebris_01}}
]
properties {
 attenuationSettings=&Cube_Roll_Att
 bOverrideConcurrency=true
 maxDistance=30000
 soundClassObject=&Fort_Mission_Ingame_Important_PSM
 volumeMultiplier=0.8
 concurrencyOverrides {maxCount=1}
}
