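# Sound cue for the Cube's final roll impact at its destination, layering close and distant sounds.
# A mixer combines a destination impact wave under close-range attenuation with a slightly boosted random pick of two distant roll-end waves under distant attenuation.
# The cue routes to an important in-game mission sound class and overrides concurrency to allow a single instance.
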SoundCue /Figment_S05_Cube/Athena/Sounds/Events/Cube/Cube_Roll_End_Dest_Cue
root=Mixer_0
nodes [
 {
  id=Attenuation_0
  children [WavePlayer_2]
  properties {attenuationSettings=&/Cube_Roll_Impact_Close_Att}
 }
 {
  id=Attenuation_1
  children [Random_2]
  properties {attenuationSettings=&/Cube_Roll_Impact_Distant_Att}
 }
 {id=Mixer_0, children [Attenuation_0, Attenuation_1], properties {inputVolume [1,1.25]}}
 {id=Random_2, children [WavePlayer_4,WavePlayer_5], properties {weights [1,1]}}
 {id=WavePlayer_2, properties {soundWaveAssetPtr=/Figment_S05_Cube/Athena/Sounds/Events/Cube/Destination/Cube_DestImpact_01}}
 {id=WavePlayer_4, properties {soundWaveAssetPtr=&/Cube_RollEndDistant_01}}
 {id=WavePlayer_5, properties {soundWaveAssetPtr=&/Cube_RollEndDistant_02}}
]
properties {
 attenuationSettings=/Game/Athena/Sounds/Events/Cube/Cube_Roll_Att
 bOverrideConcurrency=true
 maxDistance=30000
 soundClassObject=&Fort_Mission_Ingame_Important_PSM
 volumeMultiplier=0.8
 concurrencyOverrides {maxCount=1}
}
