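# Sound cue for the Cube spawn event in the Figment S05 Cube content, playing the spawn sound.
# A single wave player node feeds Cube_Spawn_01 straight to the output, with slightly reduced volume.
# It borrows the crack lightning event's attenuation and uses an important in-game mission sound class.
# A concurrency override limits it to one playing instance at a time.
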SoundCue /Figment_S05_Cube/Athena/Sounds/Events/Cube/Cube_Spawn_Cue
root=WavePlayer_2
nodes [{id=WavePlayer_2, properties {soundWaveAssetPtr=&/Cube_Spawn_01}}]
properties {
 attenuationSettings=/FNGameplayCues/Athena/Sounds/Events/CrackLightning/Crack_Lightning_Att
 bOverrideConcurrency=true
 maxDistance=2097152
 soundClassObject=&Fort_Mission_Ingame_Important_PSM
 volumeMultiplier=0.8
 concurrencyOverrides {maxCount=1}
}
